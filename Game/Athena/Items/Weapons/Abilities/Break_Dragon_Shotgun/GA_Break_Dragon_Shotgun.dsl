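; GA_Break_Dragon_Shotgun is a gameplay ability, apparently for the Dragon Shotgun weapon, that notifies the shooter when it commits.
; On CommitExecute, it validates the avatar actor and sends it an Event.Weapon.Fired gameplay event so the shooter runs fire traces for placement.
; It then calls the parent implementation.
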(fn CommitExecute
  (bind _returnvalue (Ability|GetAvatarActorfromActorInfo))
  ;[1 1 1] Send a message to the shooter so that they do the fire traces for placement
  (Utilities|IsValid _returnvalue
    (:Is%20Valid
      (Ability|SendGameplayEventtoActor _returnvalue "(TagName=\"Event.Weapon.Fired\")"))
    (:Is%20Not%20Valid
      _))
  (|CallParentFunction))
