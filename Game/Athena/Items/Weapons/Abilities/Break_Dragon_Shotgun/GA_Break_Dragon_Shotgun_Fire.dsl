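; Gameplay ability for the Dragon's Breath shotgun's fire effect, triggered by event and run on the server when Curie is active for the avatar.
; It line-traces a fan from a shoulder socket, adjusted for crouch, rotating the targeting direction by each entry in an angle array.
; Hits feed either Curie fire application or spawning of a fire fuel actor, whose very short lifespan limits problems from fuel that fails to stick.
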(event Ability|EventActivateAbilityFromEvent
  (params (EventData :type Gameplay%20Event%20Data%20Structure%20%28by%20ref%29))
  (bind _returnvalue (Ability|GetAvatarActorfromActorInfo))
  ;[1 1 1] On the server, calculate locations and do an array of traces, if the traces hit a player then do nothing but if they hit the world spawn the fire fuel actor.
  (if (Networking|IsServer)
    (Utilities|IsValid _returnvalue
      (:Is%20Valid
        (if (Curie|IsCurieActiveforActor _returnvalue)
          (bind _asfort_player_pawn (Utilities|Casting|CastToFortPlayerPawn _returnvalue))
          (bind _output_get (Variables|Default|SetPlayerPawn _asfort_player_pawn))
          (Utilities|Array|AppendArray (Variables|Default|GetIgnoredActors) (Class|FortPawn|GetProjectileIgnoredActors _output_get))
          (Utilities|Array|AppendArray (Variables|Default|GetIgnoredActors) (Utilities|Array|MakeArray _output_get))
          (for (_ _array_index) (Variables|Default|GetAngles)
            (bind (_outhit _returnvalue_2) (Collision|LineTraceByChannel (Math|Vector|vector+vector (Variables|Default|GetTraceHeightOffset) (* (Math|Vector|Normalize (Transformation|GetActorForwardVector _output_get)) (Variables|Default|GetTraceForwardOffset)) (CallFunction|GetStartTraceLocation _output_get)) (+ (Math|Vector|vector+vector (Variables|Default|GetTraceHeightOffset) (* (Math|Vector|Normalize (Transformation|GetActorForwardVector _output_get)) (Variables|Default|GetTraceForwardOffset)) (CallFunction|GetStartTraceLocation _output_get)) (CallFunction|GetTraceAngleOffset _output_get _array_index)) "TraceTypeQuery1" false (Variables|Default|GetIgnoredActors)))
            (if _returnvalue_2
              (bind _aspawn (Utilities|Casting|CastToPawn (.HitActor (Collision|BreakHitResult _outhit)))
                (:then
                  _)
                (:CastFailed
                  _))
              (bind _data (Utilities|FlowControl|MergeExecutionwithData (Pawn|GetNavAgentLocation _aspawn) (.Normal (Collision|BreakHitResult _outhit)))
                (:Then
                  (if (not (Utilities|Array|ContainsItem (Variables|Default|GetIgnoredActors) (.HitActor (Collision|BreakHitResult _outhit))))
                    (Curie|ExecuteFireApplication (Variables|Default|GetPlayerPawn) (Variables|Default|GetPlayerPawn) (.HitActor (Collision|BreakHitResult _outhit)) 1e+02 _outhit -1 256.0)
                    (Utilities|IsValid _output_get
                      (:Is%20Valid
                        (bind _returnvalue_3 (Building|SpawnBuildingGameplayActor (Variables|Default|GetFireBGA) (Math|Transform|MakeTransform _data (Math|Rotator|MakeRotfromX (.Normal (Collision|BreakHitResult _outhit)))) _output_get))
                        ;[1 1 1] Set life span low
                        ; This will help combat fuel that won't stick and instead will fall to the ground and cause issues
                        (bind _asbga_generic_curie_fuel_parent (Utilities|Casting|CastToBGA_GenericCurieFuel_Parent _returnvalue_3))
                        (Actor|SetLifeSpan _asbga_generic_curie_fuel_parent 0.1)))))))))))))

(fn GetStartTraceLocation
  (params (Pawn :type Character%20Object%20Reference))
  ;[1 1 1] Get the start location which is basically the right shoulder. Offset if crouched.
  (return (select (Variables|/Script/Engine.Character|GetbIsCrouched Pawn) (- (Transformation|GetSocketLocation (Variables|/Script/Engine.Character|GetMesh Pawn) (Variables|Default|GetSpawnSocket)) (Variables|Default|GetCrouchTraceOffset)) (Transformation|GetSocketLocation (Variables|/Script/Engine.Character|GetMesh Pawn) (Variables|Default|GetSpawnSocket)))))

(fn GetTraceAngleOffset
  (params (Pawn :type Fort%20Pawn%20Object%20Reference) (AngleIndex :type Integer))
  ;[1 1 1] Calc the offset for the trace to be added back to the starting location. Change the angle based on the index of the array.
  (return (* (Math|Vector|RotateVectorAroundAxis (Math|Vector|Normalize (Math|Vector|GetForwardVector (Targeting|GetDefaultTargetingRotation Pawn))) (Utilities|Array|Get(aref) (Variables|Default|GetAngles) AngleIndex) (Transformation|GetActorUpVector Pawn)) (Variables|Default|GetFireTraceRange))))
